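# Additive, two-sided unlit material for the effect mesh on the Raspberry backpack cosmetic, usable on skeletal meshes.
# Two half-cycle-offset samples of the FX texture slide from UV1 to UV0 and crossfade, creating seamless looping flow driven by FlowSpeed.
# The blend feeds an emissive-adjustment function and a Transparency multiply.
# Separate branches compute a camera-axis offset from ZOfs and HT_CrunchVerts, and a turbulence-warped texture sample.
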Material /BRCosmetics/Accessories/FORT_Backpacks/Backpack_F_MED_Raspberry/Effect/M_Backpack_Raspberry_FX_Mesh
graph {
 properties {
  bUsedWithMeshDeformer=true
  bUsedWithSkeletalMesh=true
  blendMode=BLEND_Additive
  pixelDepthOffsetMode=PDOM_Legacy
  shadingModel=MSM_Unlit
  twoSided=true
  usedShadingModels=Unlit
 }
 nodes [
  n1 = TextureCoordinate() {coordinateIndex=1}
  n2 = TextureCoordinate()
  n3 = Time()
  n4 = Frac(None=Multiply(A=ScalarParameter() {defaultValue=0.2, parameterName=FlowSpeed}, B=@n3) {constB=0.4})
  n5 = TextureSample(UVs=LinearInterpolate(A=@n1, B=@n2, Alpha=@n4)) {samplerType=SAMPLERTYPE_LinearColor, texture=&/T_Backpack_Raspberry_FX}
  n6 = Constant() {r=0.5}
  n7 = Add(A=@n4, B=@n6)
  n8 = TextureSample(UVs=LinearInterpolate(A=@n1, B=@n2, Alpha=Frac(None=@n7))) {samplerType=SAMPLERTYPE_LinearColor, texture=&T_Backpack_Raspberry_FX}
  n9 = Subtract(A=@n4, B=@n6)
  n10 = LinearInterpolate(A=@n5.RGB, B=@n8.RGB, Alpha=Multiply(A=Abs(None=@n9)) {constB=2})
  MaterialFunctionCall(Emissive=@n10) {materialFunction=&MF_FX_EmissiveAdjustments}
  Multiply(A=@n10, B=ScalarParameter() {defaultValue=0.1, parameterName=Transparency})
  n11 = Subtract(A=WorldPosition().XYZ, B=CameraPositionWS())
  Multiply(
   A=Multiply(A=Normalize(VectorInput=@n11), B=ScalarParameter() {parameterName=ZOfs})
   B=Clamp(None=ScalarParameter() {parameterName=HT_CrunchVerts, sliderMax=1}) {minDefault=1}
  )
  n12 = TextureCoordinate()
  n13 = TextureSample(UVs=Add(A=Multiply(A=@n12) {constB=-0.6}, B=Multiply(A=@n3) {constB=0.08})) {samplerType=SAMPLERTYPE_LinearColor, texture=/Game/Effects/Fort_Effects/Textures/Plasma/T_Turbulence_AdditiveWarp_Linear}
  n14 = Subtract(A=Multiply(A=ComponentMask(None=@n13.RGB) {g=true, r=true}) {constB=0.5}) {constB=0.5}
  TextureSample(UVs=Add(A=@n12, B=Multiply(A=@n14) {constB=0.1})) {samplerType=SAMPLERTYPE_LinearColor, texture=&T_Backpack_Raspberry_FX}
 ]
}
