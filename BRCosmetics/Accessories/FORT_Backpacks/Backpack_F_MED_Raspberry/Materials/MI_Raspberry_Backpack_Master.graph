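# Master material for the Raspberry backpack cosmetic (a base Material despite its MI_ prefix), adding animated iridescent energy and sparkles to the standard character shading chain.
# Two panning noise layers in polar coordinates around CenterPoint create energy and thresholded sparkle masks.
# SpecularMasks, a sphere mask and Fresnel build lattice, core and edge masks, while view angle and time drive hue shifting.
# Results override base color, emissive and specular between character FX functions.
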Material /BRCosmetics/Accessories/FORT_Backpacks/Backpack_F_MED_Raspberry/Materials/MI_Raspberry_Backpack_Master
graph {
 properties {
  bAutomaticallySetUsageInEditor=false
  bUseMaterialAttributes=true
  bUsedWithMeshDeformer=true
  bUsedWithMorphTargets=true
  bUsedWithSkeletalMesh=true
  pixelDepthOffsetMode=PDOM_Legacy
  refractionMethod=RM_IndexOfRefraction
  shadingModel=MSM_FromMaterialExpression
  subsurfaceProfile=&SS_HeroSkin_02
  usedShadingModels=Default Lit | Subsurface | Clear Coat | Subsurface Profile | Cloth | Eye
 }
 nodes [
  n1 = MaterialFunctionCall() {materialFunction=&BaseFN_CharacterFunction_PreFX}
  n2 = GetMaterialAttributes(None=MaterialFunctionCall(
    PreskinnedPosition=@n1.Pre skinned Local Position
    Gmap=@n1.Pre skinned Local Position
    MaterialAttributes=@n1.Pre skinned Local Position
    AO=@n1.Pre skinned Local Position
   ) {materialFunction=&CharacterSkinFX}.Result) {attributeGetTypes [#1, #2, #3]}
  n3 = NamedRerouteUsage()
  n4 = MaterialFunctionCall(
   AO=@n1.AO
   Distance Blend=@n1.AO
   Fresnel=@n1.AO
   Gmap=@n1.AO
   Pre Skinned Local Position=@n1.AO
   Material=SetMaterialAttributes!(
    MaterialAttributes=@n2.MaterialAttributes
    Base Color=NamedRerouteUsage()
    Emissive Color=NamedRerouteUsage()
    Specular=LinearInterpolate(A=@n2.Specular, Alpha=@n3)
   ) {attributeSetTypes [#1, #2, #3]}
  ) {materialFunction=&BaseFN_CharacterFunction_PostFX}
  MaterialFunctionCall(In=@n4.Result) {materialFunction=&CINE_MF_CharPostFX}
  NamedRerouteDeclaration(None=ComponentMask(None=VectorParameter() {parameterName=CenterPoint, defaultValue=<rgba 0.5 0.6 0 0>}.RGB) {g=true, r=true}) {name=UV_CenterPoint, nodeColor=<rgba 1 0.5859 0 1>}
  n5 = OneMinus(None=TextureSampleParameter2D() {parameterName=SpecularMasks, samplerSource=SSM_Wrap_WorldGroupSettings}.R)
  NamedRerouteDeclaration(None=Saturate(None=ConstantBiasScale(None=@n5) {bias=-0.33, scale=2})) {name=Lattice_Mask, nodeColor=<rgba 0.6406 0 1 1>}
  n6 = Saturate(None=SphereMask(
    A=TextureCoordinate()
    B=NamedRerouteUsage()
    Radius=ScalarParameter() {defaultValue=0.4, group=Raspberry, parameterName=CenterSphereMask_Radius}
    Hardness=Constant()
   ))
  n7 = NamedRerouteUsage()
  n8 = Saturate(None=Subtract(A=@n6, B=Multiply(A=@n7) {constB=1.5}))
  n9 = Fresnel() {baseReflectFraction=0, exponent=4}
  NamedRerouteDeclaration(None=Saturate(None=Add(A=@n8, B=@n9))) {name=Core_Edge_Mask, nodeColor=<rgba 0.0547 0 1 1>}
  n10 = NamedRerouteUsage()
  n11 = NamedRerouteUsage()
  n12 = NamedRerouteUsage()
  NamedRerouteDeclaration(None=LinearInterpolate(
    A=LinearInterpolate(A=@n10, B=@n11, Alpha=@n12)
    B=VectorParameter() {group=Raspberry, parameterName=Edge_Color, defaultValue=<rgba 0.1371 0.0508 0.25 0>}.RGB
    Alpha=NamedRerouteUsage()
   )) {name=BaseColor_Out, nodeColor=<rgba 0 1 0.3906 1>}
  n13 = NamedRerouteUsage()
  n14 = NamedRerouteUsage()
  n15 = NamedRerouteUsage()
  n16 = Add(
   A=Multiply(A=NamedRerouteUsage(), B=ScalarParameter() {defaultValue=75, group=Raspberry, parameterName=Sparkle_Intensity}) {constB=100}
   B=Multiply(A=Multiply(A=@n13, B=@n14), B=Add(A=@n15))
  )
  n17 = VectorParameter() {group=Raspberry, parameterName=Edge_Color, defaultValue=<rgba 0.1371 0.0508 0.25 0>}
  n18 = NamedRerouteUsage()
  NamedRerouteDeclaration(None=Add(A=LinearInterpolate(A=@n16, B=@n17.RGB, Alpha=@n18), B=NamedRerouteUsage())) {name=Emissive Out, nodeColor=<rgba 0 0.625 1 1>}
  n19 = NamedRerouteUsage()
  n20 = Desaturation(None=NamedRerouteUsage(), Fraction=Constant() {r=-2})
  NamedRerouteDeclaration(None=Add(
    A=MaterialFunctionCall(Hue Shift Percentage=@n19, Texture=@n20) {materialFunction=&HueShift}.Result
    B=NamedRerouteUsage()
   )) {name=Iridecent_Color, nodeColor=<rgba 0 0.4375 1 1>}
  n21 = ScalarParameter() {defaultValue=0.5, group=Raspberry, parameterName=Noise_Center_Raidal_Tile}
  n22 = Add(A=TextureCoordinate(), B=Multiply(A=NamedRerouteUsage()) {constB=-1}) {constB=0}
  n23 = ComponentMask(None=@n22) {r=true}
  n24 = Multiply(A=@n21, B=@n23)
  n25 = ComponentMask(None=@n22) {g=true}
  n26 = Multiply(A=@n21, B=@n25)
  n27 = SquareRoot(None=Add(A=Multiply(A=@n24, B=@n24), B=Multiply(A=@n26, B=@n26)))
  n28 = Divide(
   A=Arctangent2Fast(Y=@n25, X=@n23)
   B=MaterialFunctionCall(Multiplier=ScalarParameter() {defaultValue=0.5, group=Raspberry, parameterName=Noise_Center_Angle_Tile}) {desc=Radial Gradient, materialFunction=&Pi}.Result
  )
  n29 = VectorParameter() {group=Raspberry, parameterName=EffectSpeed, defaultValue=<rgba -0.125 0 0 0>}
  n30 = TextureSample(UVs=Panner(Coordinate=AppendVector(A=@n27, B=@n28), Speed=ComponentMask(None=@n29.RGB) {g=true, r=true}) {speedX=1, speedY=1}) {samplerSource=SSM_Wrap_WorldGroupSettings, samplerType=SAMPLERTYPE_Masks, texture=/Game/Weapons/WeaponSkins/Wraps/Textures/Mash/Mash_Noise}
  n31 = Power(Base=@n30.G) {constExponent=1.2}
  NamedRerouteDeclaration(None=Saturate(None=Multiply(A=@n31) {constB=2})) {name=Adjusted_Energy_Mask, nodeColor=<rgba 0 1 0.2031 1>}
  n32 = Multiply(A=NamedRerouteUsage(), B=OneMinus(None=@n6))
  n33 = Multiply(A=NamedRerouteUsage(), B=NamedRerouteUsage())
  NamedRerouteDeclaration(None=Saturate(None=Add(A=@n32, B=@n33))) {name=Combined_Effect_Mask, nodeColor=<rgba 0 0.625 1 1>}
  n34 = DotProduct(A=CameraVectorWS(), B=PixelNormalWS())
  n35 = Time()
  NamedRerouteDeclaration(None=Add(A=Multiply(A=@n34) {constB=2}, B=Multiply(A=@n35) {constB=0.33})) {name=Hue_Shift_Amount, nodeColor=<rgba 0.3828 0 1 1>}
  NamedRerouteDeclaration(None=@n30.G) {name=Energy_Mask, nodeColor=<rgba 0.1016 1 0 1>}
  n36 = ScalarParameter() {defaultValue=4, group=Raspberry, parameterName=Noise_Center_Raidal_Tile_B}
  n37 = Add(A=TextureCoordinate(), B=Multiply(A=NamedRerouteUsage()) {constB=-1}) {constB=0}
  n38 = ComponentMask(None=@n37) {r=true}
  n39 = Multiply(A=@n36, B=@n38)
  n40 = ComponentMask(None=@n37) {g=true}
  n41 = Multiply(A=@n36, B=@n40)
  n42 = Add(A=Multiply(A=@n39, B=@n39), B=Multiply(A=@n41, B=@n41))
  n43 = Arctangent2Fast(Y=@n40, X=@n38)
  n44 = MaterialFunctionCall(Multiplier=ScalarParameter() {defaultValue=0.075, group=Raspberry, parameterName=Noise_Center_Angle_Tile_B}) {desc=Radial Gradient, materialFunction=&Pi}
  n45 = Panner(
   Coordinate=AppendVector(A=SquareRoot(None=@n42), B=Divide(A=@n43, B=@n44.Result))
   Speed=ComponentMask(None=VectorParameter() {group=Raspberry, parameterName=EffectSpeed_B, defaultValue=<rgba -0.4 0 0 0>}.RGB) {g=true, r=true}
  ) {speedX=1, speedY=1}
  n46 = Ceil(None=Subtract(A=TextureSample(UVs=@n45) {samplerSource=SSM_Wrap_WorldGroupSettings, texture=/Game/Global/Textures/Default/Noise/Noise}.R) {constB=0.7})
  n47 = NamedRerouteUsage()
  n48 = NamedRerouteUsage()
  n49 = Constant3Vector() {constant=<rgba 1 0 0 0>}
  NamedRerouteDeclaration(None=Multiply(
    A=Multiply(A=@n46, B=@n47)
    B=MaterialFunctionCall(Hue Shift Percentage=@n48, Texture=@n49) {materialFunction=&HueShift}.Result
   )) {name=Sparkle_Mask, nodeColor=<rgba 0 1 0.25 1>}
  NamedRerouteDeclaration(None=@n2.Base Color) {name=BaseColor_In, nodeColor=<rgba 0.3828 1 0 1>}
  NamedRerouteDeclaration(None=@n2.Emissive Color) {name=Emissive_In, nodeColor=<rgba 0.0547 0 1 1>}
 ]
}
